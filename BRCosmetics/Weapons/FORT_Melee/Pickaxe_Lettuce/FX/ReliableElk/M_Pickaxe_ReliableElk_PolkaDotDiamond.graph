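# Translucent unlit Niagara sprite material for the ReliableElk pickaxe effect, drawing a glowing ring filled with polka dots or harlequin diamonds.
# A radial band driven by the Glow Torus Radius dynamic parameter scales dot size and diamond threshold, so both patterns vary across the ring.
# A parameter blends the two patterns, which receive an edge fade and added base glow.
# Particle color tints the emissive, while opacity gets an optional circular cutout and depth fade.
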Material /BRCosmetics/Weapons/FORT_Melee/Pickaxe_Lettuce/FX/ReliableElk/M_Pickaxe_ReliableElk_PolkaDotDiamond
graph {
 properties {bUsedWithNiagaraSprites=true, blendMode=BLEND_Translucent, pixelDepthOffsetMode=PDOM_Legacy, shadingModel=MSM_Unlit, usedShadingModels=Unlit}
 nodes [
  n1 = ParticleColor()
  n2 = Frac(None=Add(A=TextureCoordinate()) {constB=0.5})
  n3 = Constant() {r=0.5}
  n4 = ScalarParameter() {defaultValue=0.75, parameterName=Radius}
  n5 = ScalarParameter() {defaultValue=0.5, parameterName=Hardness, sliderMax=1}
  n6 = Power(
   Base=OneMinus(None=SphereMask(A=@n2, B=@n3, Radius=@n4, Hardness=@n5) {hardnessPercent=75})
   Exp=ScalarParameter() {defaultValue=1, parameterName=Fade Exp, sliderMax=10}
  )
  n7 = TextureCoordinate()
  n8 = ScalarParameter() {defaultValue=10, parameterName=Polka Dot Tile, sliderMax=20}
  n9 = DynamicParameter() {defaultValue=<rgba 0.25 0.2 1 1>, paramNames [Glow Torus Radius, Glow Torus Width, Param3, Param4]}
  n10 = SphereMask(A=Distance(A=TextureCoordinate(), B=Constant() {r=0.5}), B=@n9.Glow Torus Radius, Radius=@n9.Glow Torus Radius) {hardnessPercent=0}
  n11 = SphereMask(
   A=Frac(None=Multiply(A=@n7, B=@n8) {constB=10})
   B=Constant() {r=0.5}
   Radius=Multiply(A=ScalarParameter() {defaultValue=0.4, parameterName=Polka Dot Size}, B=@n10)
   Hardness=ScalarParameter() {parameterName=Polka Dot Hardness, sliderMax=1}
  )
  n12 = Multiply(A=TextureCoordinate(), B=Constant2Vector() {g=4, r=6}) {constB=6}
  n13 = ScalarParameter() {defaultValue=2.5, parameterName=HarlequinPatternTile}
  n14 = Subtract(A=Frac(None=Multiply(A=@n12, B=@n13) {constB=3})) {constB=0.5}
  n15 = Abs(None=ComponentMask(None=@n14) {r=true})
  n16 = Abs(None=ComponentMask(None=@n14) {g=true})
  n17 = LinearInterpolate(Alpha=@n10) {constA=0.5, constB=0}
  n18 = Round(None=Subtract(A=Add(A=@n15, B=@n16), B=Saturate(None=@n17)) {constB=0})
  n19 = Multiply(
   A=@n6
   B=LinearInterpolate(A=Multiply(A=@n11, B=@n10), B=Multiply(A=@n10, B=@n18), Alpha=ScalarParameter() {parameterName=UseHarlequinPattern})
  )
  n20 = Add(
   A=@n19
   B=Multiply(A=Multiply(A=@n6, B=@n10), B=ScalarParameter() {defaultValue=0.1, parameterName=Base Glow Intensity, sliderMax=1}) {constB=0.1}
  )
  n21 = Multiply(
   A=@n20
   B=MaterialFunctionCall!(UVs, CenterPosition, Radius=Constant() {r=0.46}) {materialFunction=/Engine/Functions/Engine_MaterialFunctions01/Gradient/SphereGradient-2D}.Result 0-1
  )
  n22 = ScalarParameter() {parameterName=UseAlphaCutout?}
  DepthFade(
   Opacity=Multiply(A=@n1.A, B=LinearInterpolate(A=@n20, B=@n21, Alpha=@n22))
   FadeDistance=ScalarParameter() {defaultValue=30, parameterName=Depth Fade Distance, sliderMax=100}
  )
  MaterialFunctionCall(Emissive=Multiply(A=@n1.RGB, B=Add(B=@n19) {constA=1})) {materialFunction=&MF_FX_EmissiveAdjustments}
 ]
}
