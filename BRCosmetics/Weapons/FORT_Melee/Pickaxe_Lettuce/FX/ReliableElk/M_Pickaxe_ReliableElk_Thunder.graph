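# Unlit translucent material for Niagara sprite particles in the thunder effect of the ReliableElk pickaxe cosmetic.
# The particle color RGB goes through MF_FX_EmissiveAdjustments to produce the emissive glow.
# Opacity is particle alpha times the red channel of the thunder texture, softened against scene geometry by a depth fade with an adjustable distance parameter.
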Material /BRCosmetics/Weapons/FORT_Melee/Pickaxe_Lettuce/FX/ReliableElk/M_Pickaxe_ReliableElk_Thunder
graph {
 properties {bUsedWithNiagaraSprites=true, blendMode=BLEND_Translucent, pixelDepthOffsetMode=PDOM_Legacy, shadingModel=MSM_Unlit, usedShadingModels=Unlit}
 nodes [
  n1 = ParticleColor()
  n2 = TextureSample() {texture=&/T_Pickaxe_ReliableElk_Thunder}
  DepthFade(
   Opacity=Multiply(A=@n1.A, B=Max(A=@n2.R, B=@n2.R))
   FadeDistance=ScalarParameter() {defaultValue=35, parameterName=Depth Fade Distance, sliderMax=100}
  )
  MaterialFunctionCall(Emissive=@n1.RGB) {materialFunction=&MF_FX_EmissiveAdjustments}
 ]
}
